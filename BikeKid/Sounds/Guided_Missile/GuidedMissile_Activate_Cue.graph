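# Sound cue for the guided missile activation sound, routed to a third-person ranged weapon fire sound class.
# A Random node picks one of three equally weighted activation waves and feeds a Modulator at the root for pitch and volume variation.
# Attenuation overrides replace the referenced missile launch attenuation asset, adding long-range falloff with distance-based low-pass and high-pass filtering on a logarithmic frequency scale.
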SoundCue /BikeKid/Sounds/Guided_Missile/GuidedMissile_Activate_Cue
root=Modulator_0
nodes [
 {id=Modulator_0, children [Random_0]}
 {id=Random_0, children [WavePlayer_0,WavePlayer_4,WavePlayer_5], properties {bShouldExcludeFromBranchCulling=true, weights [1,1,1]}}
 {id=WavePlayer_0, properties {soundWaveAssetPtr=&/GuidedMissile_Activate_03}}
 {id=WavePlayer_4, properties {soundWaveAssetPtr=&/GuidedMissile_Activate_02}}
 {id=WavePlayer_5, properties {soundWaveAssetPtr=&/GuidedMissile_Activate_01}}
]
properties {
 attenuationSettings=/Game/Sounds/Attenuation/Fort_Weapon_Projectile_Launch_Missile_Athena_Att
 bOverrideAttenuation=true
 maxDistance=13545.416
 soundClassObject=&Fort_WeaponRanged_Fire_3P_PSM
 volumeMultiplier=1
 attenuationOverrides {
  bAttenuateWithLPF=true
  bEnableLogFrequencyScaling=true
  falloffDistance=13139.8223
  hPFFrequencyAtMax=175.0834
  lPFFrequencyAtMax=771.233
  lPFRadiusMax=9108.7695
  attenuationShapeExtents {x=405.5941}
 }
}
