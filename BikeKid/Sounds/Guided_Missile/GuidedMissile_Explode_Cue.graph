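# Sound cue for a guided missile explosion, mixing a close layer with a louder-weighted distant layer.
# The close layer randomly picks one of three explosion waves, modulates it and uses short-range attenuation.
# The distant layer plays one sharp distant explosion with pitch variation and much longer falloff.
# Both layers override to a third-person ranged weapon fire sound class, and concurrency keeps one instance per owner, stopping the oldest.
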SoundCue /BikeKid/Sounds/Guided_Missile/GuidedMissile_Explode_Cue
root=Mixer_4
nodes [
 {
  id=Attenuation_4
  children [Modulator_0]
  properties {attenuationSettings=/Game/Sounds/Attenuation/Fort_Weapon_Rocket_Medium_Explosion_Athena_Att, attenuationOverrides {falloffDistance=1600}}
 }
 {
  id=Attenuation_5
  children [Modulator_5]
  properties {
   attenuationSettings=/Game/Sounds/Attenuation/Fort_Weapon_Rocket_Distant_Explosion_Athena_Att
   attenuationOverrides {falloffDistance=20000, attenuationShapeExtents {x=1000}}
  }
 }
 {id=Mixer_4, children [SoundClass_3,SoundClass_2], properties {inputVolume [0.9,1.2]}}
 {id=Modulator_0, children [Random_0]}
 {id=Modulator_5, children [WavePlayer_0], properties {pitchMax=1.1, pitchMin=0.9}}
 {id=Random_0, children [WavePlayer_4,WavePlayer_8,WavePlayer_9], properties {bShouldExcludeFromBranchCulling=true, weights [1,1,1]}}
 {
  id=SoundClass_2
  children [Attenuation_5]
  properties {soundClassOverride=&Fort_WeaponRanged_Fire_3P_PSM}
 }
 {
  id=SoundClass_3
  children [Attenuation_4]
  properties {soundClassOverride=&Fort_WeaponRanged_Fire_3P_PSM}
 }
 {id=WavePlayer_0, properties {soundWaveAssetPtr=/Game/Sounds/Fort_GamePlay_Sounds/Explosives/explosion_sharp_distant_01}}
 {id=WavePlayer_4, properties {soundWaveAssetPtr=&/GuidedMissile_Explode_01}}
 {id=WavePlayer_8, properties {soundWaveAssetPtr=&/GuidedMissile_Explode_03}}
 {id=WavePlayer_9, properties {soundWaveAssetPtr=&/GuidedMissile_Explode_02}}
]
properties {
 maxDistance=15000
 volumeMultiplier=1
 attenuationOverrides {falloffDistance=20000}
 concurrencySet [/Game/Sounds/Concurrency_Rules/OneMaximum_Owner_StopOldest]
}
