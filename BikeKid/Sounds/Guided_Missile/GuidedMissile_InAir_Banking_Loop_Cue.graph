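# Looping Sound Cue for the banking sound of an in-air guided missile.
# A BankingSound parameter crossfades from loop 04, at reduced volume, to loop 01 at full volume across the middle of its range.
# Each branch passes through close missile projectile attenuation and a third-person ranged weapon fire sound class override.
# The cue references a one-per-owner, stop-oldest concurrency rule.
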SoundCue /BikeKid/Sounds/Guided_Missile/GuidedMissile_InAir_Banking_Loop_Cue
root=ParamCrossFade_1
nodes [
 {
  id=Attenuation_1
  children [WavePlayer_1]
  properties {attenuationSettings=/Game/Sounds/Attenuation/Fort_Weapon_Projectile_Close_Missile_Athena_Att}
 }
 {
  id=Attenuation_2
  children [WavePlayer_2]
  properties {attenuationSettings=&Fort_Weapon_Projectile_Close_Missile_Athena_Att}
 }
 {
  id=ParamCrossFade_1
  children [SoundClass_1,SoundClass_3]
  properties {
   paramName=BankingSound
   crossFadeInput [
    {fadeInDistanceEnd=0.7, fadeInDistanceStart=0.3, fadeOutDistanceEnd=1, fadeOutDistanceStart=1, volume=1}
    {fadeInDistanceEnd=0, fadeInDistanceStart=0, fadeOutDistanceEnd=0.7, fadeOutDistanceStart=0.3, volume=0.7}
   ]
  }
 }
 {
  id=SoundClass_1
  children [Attenuation_1]
  properties {soundClassOverride=&Fort_WeaponRanged_Fire_3P_PSM}
 }
 {
  id=SoundClass_3
  children [Attenuation_2]
  properties {soundClassOverride=&Fort_WeaponRanged_Fire_3P_PSM}
 }
 {id=WavePlayer_1, properties {bLooping=true, soundWaveAssetPtr=&/GuidedMissile_InAir_Banking_Loop_01}}
 {id=WavePlayer_2, properties {bLooping=true, soundWaveAssetPtr=&/GuidedMissile_InAir_Banking_Loop_04}}
]
properties {
 maxDistance=4500
 soundClassObject=&Fort_WeaponRanged_Fire_3P_PSM
 volumeMultiplier=0.55
 concurrencyOverrides {bLimitToOwner=true, maxCount=1, resolutionRule=PreventNew}
 concurrencySet [/Game/Sounds/Concurrency_Rules/OneMaximum_Owner_StopOldest]
}
