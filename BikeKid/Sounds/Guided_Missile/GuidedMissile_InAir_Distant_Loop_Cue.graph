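# Sound cue for the distant in-air loop of a guided missile projectile.
# A looping wave player feeds a Modulator that randomizes pitch and volume, then an Attenuation node using distant rocket settings.
# The root SoundClass node overrides the cue's third-person weapon fire class with the distant fire class.
# The cue scales volume down and overrides its assigned attenuation asset with inline falloff and low-pass filtering.
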SoundCue /BikeKid/Sounds/Guided_Missile/GuidedMissile_InAir_Distant_Loop_Cue
root=SoundClass_6
nodes [
 {id=Attenuation_11, children [Modulator_7], properties {attenuationSettings=/Game/Sounds/Attenuation/Fort_Weapon_Rocket_Distant_Athena_Att}}
 {id=Modulator_7, children [WavePlayer_0], properties {pitchMax=1.1, pitchMin=0.8, volumeMax=1.5, volumeMin=1.3}}
 {
  id=SoundClass_6
  children [Attenuation_11]
  properties {soundClassOverride=&Fort_WeaponRanged_Fire_Distant}
 }
 {
  id=WavePlayer_0
  properties {bLooping=true, soundWaveAssetPtr=/Game/Athena/Sounds/Weapons/Guided_Missile/Unvault/GuidedMissile_Unvault_InAir_Distant_Loop_01}
 }
]
properties {
 attenuationSettings=/Game/Sounds/Fort_Weapons/Launchers/Launcher_Prj_Loop_Att
 bOverrideAttenuation=true
 maxDistance=15000
 soundClassObject=&Fort_WeaponRanged_Fire_3P_PSM
 volumeMultiplier=0.2
 attenuationOverrides {bAttenuateWithLPF=true, falloffDistance=13084.043}
}
